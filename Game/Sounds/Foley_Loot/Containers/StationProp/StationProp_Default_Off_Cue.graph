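# Sound cue for the default station prop loot container's "off" foley sound.
# A single wave player feeds a modulator that randomizes pitch and volume on each play.
# It is scaled down in volume, uses a shared StationProp attenuation asset and the loot reward sound class, and overrides concurrency to allow two instances, stopping the farthest first.
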SoundCue /Game/Sounds/Foley_Loot/Containers/StationProp/StationProp_Default_Off_Cue
root=Modulator_0
nodes [
 {id=Modulator_0, children [WavePlayer_0], properties {pitchMax=1.3, pitchMin=0.8, volumeMax=1, volumeMin=0.6}}
 {id=WavePlayer_0, properties {soundWaveAssetPtr=&/StationProp_Default_Off_01}}
]
properties {
 attenuationSettings=&/StationProp_Default_Att
 bOverrideConcurrency=true
 maxDistance=6300
 soundClassObject=&Fort_Loot_Reward_PSM
 volumeMultiplier=0.3
 concurrencyOverrides {maxCount=2, resolutionRule=StopFarthestThenPreventNew}
}
