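# Sound cue for the switch-on foley of the default StationProp loot container, routed to the Fort_Loot_Reward_PSM sound class.
# A quality-level root sends its first two levels to an equal-weight random pick of three wave variations and the other two to a single fixed wave.
# The random node is excluded from branch culling.
# Shared attenuation shapes falloff, and a concurrency override caps simultaneous instances, stopping the farthest first.
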SoundCue /Game/Sounds/Foley_Loot/Containers/StationProp/StationProp_Default_On_Cue
root=QualityLevel_0
nodes [
 {id=QualityLevel_0, children [Random_0,Random_0,WavePlayer_4,WavePlayer_4]}
 {id=Random_0, children [WavePlayer_1,WavePlayer_2,WavePlayer_3], properties {bShouldExcludeFromBranchCulling=true, weights [1,1,1]}}
 {id=WavePlayer_1, properties {soundWaveAssetPtr=&/StationProp_Default_On_01}}
 {id=WavePlayer_2, properties {soundWaveAssetPtr=&/StationProp_Default_On_02}}
 {id=WavePlayer_3, properties {soundWaveAssetPtr=&/StationProp_Default_On_03}}
 {id=WavePlayer_4, properties {soundWaveAssetPtr=&StationProp_Default_On_01}}
]
properties {
 attenuationSettings=&/StationProp_Default_Att
 bOverrideConcurrency=true
 maxDistance=6300
 soundClassObject=&Fort_Loot_Reward_PSM
 volumeMultiplier=0.3
 concurrencyOverrides {maxCount=2, resolutionRule=StopFarthestThenPreventNew}
}
