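# StationProp_Jingle_Salesman_Cue is a Sound Cue for the salesman jingle of a station prop, filed under container loot foley.
# A single Wave Player feeds the StationProp_Jingle_Salesman wave straight to the output, with no randomisation or modulation.
# It uses the shared StationProp default attenuation and the loot reward sound class.
# A concurrency override caps simultaneous instances, and the cue volume is slightly reduced.
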SoundCue /Game/Sounds/Foley_Loot/Containers/StationProp/StationProp_Jingle_Salesman_Cue
root=WavePlayer_0
nodes [{id=WavePlayer_0, properties {soundWaveAssetPtr=&/StationProp_Jingle_Salesman}}]
properties {
 attenuationSettings=&/StationProp_Default_Att
 bOverrideConcurrency=true
 maxDistance=6300
 soundClassObject=&Fort_Loot_Reward_PSM
 volumeMultiplier=0.848
 concurrencyOverrides {maxCount=3}
}
